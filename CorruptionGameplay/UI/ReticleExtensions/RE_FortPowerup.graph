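# Reticle extension widget that shows a circular heat meter for the equipped ranged powerup weapon, with Normal, Superheated and Overheated states.
# It binds to the weapon's overheat-value event, feeding fill, glow and threshold masks into dynamic materials.
# The same callback compares heat to SuperheatTriggerPercent and fires state-change events that toggle visibility, tint and animations.
# Firing while Superheated increments a shot counter and plays shake and lightning animations.
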WidgetBlueprint /CorruptionGameplay/UI/ReticleExtensions/RE_FortPowerup
animations [
 SupershotResponse {
  binding_count=1
  length_seconds=0.5
  object=&self:SupershotResponse
  sequence {bindings [SuperheatedShotFired_Overlay {
     id=#1
     possessed_class=Overlay
     tracks [
      {
       type=MovieSceneFloatTrack
       display_name=Render Opacity
       sections [{type=MovieSceneFloatSection, end_frame=10, start_frame=0, channels [None {type=MovieSceneScriptingFloatChannel, keys [[0,1],[5,1],[10,0]]}]}]
      }
      {
       type=MovieScene2DTransformTrack
       sections [{
         type=MovieScene2DTransformSection
         end_frame=10
         start_frame=0
         channels [
          Translation.X {type=MovieSceneScriptingFloatChannel}
          Translation.Y {type=MovieSceneScriptingFloatChannel}
          Angle {type=MovieSceneScriptingFloatChannel}
          Scale.X {type=MovieSceneScriptingFloatChannel, keys [[0,0.7],[3,1]]}
          Scale.Y {type=MovieSceneScriptingFloatChannel, keys [[0,0.7],[3,1]]}
          Shear.X {type=MovieSceneScriptingFloatChannel}
          Shear.Y {type=MovieSceneScriptingFloatChannel}
         ]
        }]
      }
     ]
    }]}
 }
 OnThresholdReached {
  binding_count=5
  length_seconds=0.15
  object=&self:OnThresholdReached
  sequence {bindings [
    Overlay_Body {
     id=#2
     possessed_class=Overlay
     tracks [{
       type=MovieScene2DTransformTrack
       sections [{
         type=MovieScene2DTransformSection
         channels [
          Translation.X {type=MovieSceneScriptingFloatChannel, keys [[0,10],[3,20]]}
          Translation.Y {type=MovieSceneScriptingFloatChannel, keys [[0,10],[3,0]]}
          Angle {type=MovieSceneScriptingFloatChannel, keys [[0,0],[3,-45]]}
          Scale.X {type=MovieSceneScriptingFloatChannel, keys [[0,1],[3,1.5]]}
          Scale.Y {type=MovieSceneScriptingFloatChannel, keys [[0,1],[3,1.5]]}
         ]
        }]
      }]
    }
    HeatZoneIndicator_TopTick {
     id=#3
     possessed_class=Image
     tracks [{
       type=MovieSceneWidgetMaterialTrack
       sections [{type=MovieSceneParameterSection, channels [AnimateOverallAlpha {type=MovieSceneScriptingFloatChannel, keys [[0,0],[3,1]]}]}]
      }]
    }
    HeatZoneIndicator_BottomTick {
     id=#4
     possessed_class=Image
     tracks [{
       type=MovieSceneWidgetMaterialTrack
       sections [{type=MovieSceneParameterSection, channels [AnimateOverallAlpha {type=MovieSceneScriptingFloatChannel, keys [[0,0],[3,1]]}]}]
      }]
    }
    HeatZoneIndicator {
     id=#5
     possessed_class=Image
     tracks [{
       type=MovieSceneWidgetMaterialTrack
       sections [{type=MovieSceneParameterSection, channels [AnimateOverallAlpha {type=MovieSceneScriptingFloatChannel, keys [[0,0],[3,1]]}]}]
      }]
    }
    OverlayContent {
     id=#6
     possessed_class=Overlay
     tracks [{
       type=MovieSceneFloatTrack
       display_name=Render Opacity
       sections [{type=MovieSceneFloatSection, end_frame=3, start_frame=0, channels [None {type=MovieSceneScriptingFloatChannel, keys [[0,0.666],[3,0.75]]}]}]
      }]
    }
   ]}
 }
 Overheated {
  binding_count=3
  length_seconds=1
  object=&self:Overheated
  sequence {
   bindings [
    Backing_Overheated {
     id=#7
     possessed_class=Image
     tracks [{
       type=MovieSceneWidgetMaterialTrack
       sections [{
         type=MovieSceneParameterSection
         channels [
          ShouldHeatUp {type=MovieSceneScriptingFloatChannel, keys [[0,0],[15,1],[30,0],[45,1],[60,0]]}
          AnimateOverallAlpha {type=MovieSceneScriptingFloatChannel, keys [[0,0],[15,1],[30,0],[45,1],[60,0]]}
         ]
        }]
      }]
    }
    Image_2 {
     id=#8
     possessed_class=Image
     tracks [{
       type=MovieSceneWidgetMaterialTrack
       sections [{
         type=MovieSceneParameterSection
         channels [Final_Alpha {type=MovieSceneScriptingFloatChannel, keys [[0,0.333],[15,0.666],[30,0.333],[45,0.666],[60,0.333]]}]
        }]
      }]
    }
    BatteryIconWarningGlow {
     id=#9
     possessed_class=Image
     tracks [{
       type=MovieSceneFloatTrack
       display_name=Render Opacity
       sections [{type=MovieSceneFloatSection, channels [None {type=MovieSceneScriptingFloatChannel, keys [[0,0.5],[15,1],[30,0.5],[45,1],[60,0.5]]}]}]
      }]
    }
   ]
   display_rate {numerator=60}
  }
 }
 OnThresholdReached_NoRotation {
  binding_count=5
  length_seconds=0.1
  object=&self:OnThresholdReached_NoRotation
  sequence {bindings [
    Overlay_Body {
     id=#2
     possessed_class=Overlay
     tracks [{
       type=MovieScene2DTransformTrack
       sections [{
         type=MovieScene2DTransformSection
         channels [
          Translation.X {type=MovieSceneScriptingFloatChannel, keys [[0,10],[2,14]]}
          Translation.Y {type=MovieSceneScriptingFloatChannel, keys [[0,10],[2,14]]}
          Angle {type=MovieSceneScriptingFloatChannel, const [0,1]}
          Scale.X {type=MovieSceneScriptingFloatChannel, keys [[0,1],[2,1.1]]}
          Scale.Y {type=MovieSceneScriptingFloatChannel, keys [[0,1],[2,1.1]]}
         ]
        }]
      }]
    }
    HeatZoneIndicator_TopTick {
     id=#3
     possessed_class=Image
     tracks [{
       type=MovieSceneWidgetMaterialTrack
       sections [{type=MovieSceneParameterSection, channels [AnimateOverallAlpha {type=MovieSceneScriptingFloatChannel, keys [[0,0],[2,1]]}]}]
      }]
    }
    HeatZoneIndicator_BottomTick {
     id=#4
     possessed_class=Image
     tracks [{
       type=MovieSceneWidgetMaterialTrack
       sections [{type=MovieSceneParameterSection, channels [AnimateOverallAlpha {type=MovieSceneScriptingFloatChannel, keys [[0,0],[2,1]]}]}]
      }]
    }
    HeatZoneIndicator {
     id=#5
     possessed_class=Image
     tracks [{
       type=MovieSceneWidgetMaterialTrack
       sections [{type=MovieSceneParameterSection, channels [AnimateOverallAlpha {type=MovieSceneScriptingFloatChannel, keys [[0,0],[2,1]]}]}]
      }]
    }
    OverlayContent {
     id=#6
     possessed_class=Overlay
     tracks [{
       type=MovieSceneFloatTrack
       display_name=Render Opacity
       sections [{type=MovieSceneFloatSection, channels [None {type=MovieSceneScriptingFloatChannel, keys [[0,0.5],[2,0.75]]}]}]
      }]
    }
   ]}
 }
 ShotsFired {
  binding_count=3
  length_seconds=0.25
  object=&self:ShotsFired
  sequence {
   bindings [
    Backing {
     id=#10
     possessed_class=Image
     tracks [{
       type=MovieSceneWidgetMaterialTrack
       sections [{type=MovieSceneParameterSection, channels [RingWidth {type=MovieSceneScriptingFloatChannel, keys [[0,3],[1,4.5],[15,3]]}]}]
      }]
    }
    Backing_HeatZone {
     id=#11
     possessed_class=Image
     tracks [{
       type=MovieSceneWidgetMaterialTrack
       sections [{type=MovieSceneParameterSection, channels [RingWidth {type=MovieSceneScriptingFloatChannel, keys [[0,3],[1,4.5],[15,3]]}]}]
      }]
    }
    Progress_InThreshold {
     id=#12
     possessed_class=Image
     tracks [{
       type=MovieSceneWidgetMaterialTrack
       sections [{type=MovieSceneParameterSection, channels [RingWidth {type=MovieSceneScriptingFloatChannel, keys [[0,2.333],[1,4.5],[15,2.333]]}]}]
      }]
    }
   ]
   display_rate {numerator=60}
  }
 }
 BoostIndication {
  binding_count=6
  length_seconds=1
  object=&self:BoostIndication
  sequence {
   bindings [
    BatteryIconWarningGlow_1 {
     id=#13
     possessed_class=Image
     tracks [{
       type=MovieSceneWidgetMaterialTrack
       sections [{type=MovieSceneParameterSection, channels [Final_Alpha {type=MovieSceneScriptingFloatChannel, keys [[0,0.025],[3,0.5],[60,0.025]]}]}]
      }]
    }
    Noise {
     id=#14
     possessed_class=Image
     tracks [{
       type=MovieSceneFloatTrack
       display_name=Render Opacity
       sections [{type=MovieSceneFloatSection, channels [None {type=MovieSceneScriptingFloatChannel, keys [[0,0],[3,1],[45,0]]}]}]
      }]
    }
    PulseRinge {
     id=#15
     possessed_class=Image
     tracks [
      {
       type=MovieScene2DTransformTrack
       sections [{
         type=MovieScene2DTransformSection
         channels [
          Translation.X {type=MovieSceneScriptingFloatChannel, const [0,1]}
          Translation.Y {type=MovieSceneScriptingFloatChannel, const [0,1]}
          Angle {type=MovieSceneScriptingFloatChannel, const [0,1]}
          Scale.X {type=MovieSceneScriptingFloatChannel, keys [[0,1],[30,2.5]]}
          Scale.Y {type=MovieSceneScriptingFloatChannel, keys [[0,1],[30,2.5]]}
          Shear.X {type=MovieSceneScriptingFloatChannel, const [0,1]}
          Shear.Y {type=MovieSceneScriptingFloatChannel, const [0,1]}
         ]
        }]
      }
      {
       type=MovieSceneWidgetMaterialTrack
       sections [{type=MovieSceneParameterSection, channels [Inside_Radius {type=MovieSceneScriptingFloatChannel, keys [[0,0.333],[60,0.4]]}]}]
      }
      {
       type=MovieSceneFloatTrack
       display_name=Render Opacity
       sections [{
         type=MovieSceneFloatSection
         end_frame=60
         start_frame=0
         channels [None {type=MovieSceneScriptingFloatChannel, keys [[0,0],[1,0.5],[30,0]]}]
        }]
      }
     ]
    }
    BoostIndicator {
     id=#16
     possessed_class=Overlay
     tracks [{
       type=MovieSceneFloatTrack
       display_name=Render Opacity
       sections [{type=MovieSceneFloatSection, channels [None {type=MovieSceneScriptingFloatChannel, keys [[0,0.5],[3,1],[45,0.5]]}]}]
      }]
    }
    Ring {
     id=#17
     possessed_class=Image
     tracks [{
       type=MovieSceneFloatTrack
       display_name=Render Opacity
       sections [{type=MovieSceneFloatSection, channels [None {type=MovieSceneScriptingFloatChannel, keys [[0,0],[3,1],[45,0]]}]}]
      }]
    }
    Lightning {
     id=#18
     possessed_class=Image
     tracks [{
       type=MovieSceneFloatTrack
       display_name=Render Opacity
       sections [{type=MovieSceneFloatSection, channels [None {type=MovieSceneScriptingFloatChannel, keys [[0,0],[3,1],[45,0]]}]}]
      }]
    }
   ]
   display_rate {numerator=60}
  }
 }
 BoostIndication_Reached {
  binding_count=1
  length_seconds=0.3
  object=&self:BoostIndication_Reached
  sequence {bindings [BoostIndicator {
     id=#19
     possessed_class=Overlay
     tracks [
      {
       type=MovieScene2DTransformTrack
       sections [{
         type=MovieScene2DTransformSection
         channels [
          Translation.X {type=MovieSceneScriptingFloatChannel, keys [[1,14],[6,13]]}
          Translation.Y {type=MovieSceneScriptingFloatChannel, keys [[1,-20],[6,-13]]}
          Angle {type=MovieSceneScriptingFloatChannel, const [0,1]}
          Scale.X {type=MovieSceneScriptingFloatChannel, keys [[1,1],[6,0.75]]}
          Scale.Y {type=MovieSceneScriptingFloatChannel, keys [[1,1],[6,0.75]]}
         ]
        }]
      }
      {
       type=MovieSceneFloatTrack
       display_name=Render Opacity
       sections [{type=MovieSceneFloatSection, channels [None {type=MovieSceneScriptingFloatChannel, keys [[0,0],[1,0],[6,0.5]]}]}]
      }
     ]
    }]}
 }
 ShotsFired_MoreObvious {
  binding_count=7
  length_seconds=0.25
  object=&self:ShotsFired_MoreObvious
  sequence {
   bindings [
    OverHeatedColorBorder {
     id=#20
     possessed_class=Border
     tracks [{
       type=MovieScene2DTransformTrack
       sections [{
         type=MovieScene2DTransformSection
         end_frame=15
         start_frame=0
         channels [
          Translation.X {type=MovieSceneScriptingFloatChannel, const [0,1]}
          Translation.Y {type=MovieSceneScriptingFloatChannel, const [0,1]}
          Angle {type=MovieSceneScriptingFloatChannel, const [0,1]}
          Scale.X {type=MovieSceneScriptingFloatChannel, keys [[1,1],[2,1.035],[15,1]]}
          Scale.Y {type=MovieSceneScriptingFloatChannel, keys [[1,1],[2,1.035],[15,1]]}
          Shear.X {type=MovieSceneScriptingFloatChannel, const [0,1]}
          Shear.Y {type=MovieSceneScriptingFloatChannel, const [0,1]}
         ]
        }]
      }]
    }
    HeatZoneIndicator {
     id=#21
     possessed_class=Image
     tracks [{
       type=MovieSceneWidgetMaterialTrack
       sections [{
         type=MovieSceneParameterSection
         end_frame=15
         start_frame=0
         channels [RingWidth {type=MovieSceneScriptingFloatChannel, keys [[0,0.666],[1,2],[15,0.666]]}]
        }]
      }]
    }
    OverlayHeatzone {
     id=#22
     possessed_class=Overlay
     tracks [{
       type=MovieScene2DTransformTrack
       sections [{
         type=MovieScene2DTransformSection
         end_frame=15
         start_frame=0
         channels [
          Translation.X {type=MovieSceneScriptingFloatChannel, keys [[0,0],[1,-1],[3,3],[4,0],[5,1],[7,-2],[8,2],[10,-1],[12,1],[15,0]]}
          Translation.Y {type=MovieSceneScriptingFloatChannel, const [0,3]}
          Angle {type=MovieSceneScriptingFloatChannel, const [0,1]}
          Scale.X {type=MovieSceneScriptingFloatChannel, const [1,1]}
          Scale.Y {type=MovieSceneScriptingFloatChannel, const [1,1]}
          Shear.X {type=MovieSceneScriptingFloatChannel, const [0,1]}
          Shear.Y {type=MovieSceneScriptingFloatChannel, const [0,1]}
         ]
        }]
      }]
    }
    HeatZoneIndicator_Overheated {
     id=#23
     possessed_class=Image
     tracks [{
       type=MovieSceneWidgetMaterialTrack
       sections [{
         type=MovieSceneParameterSection
         end_frame=15
         start_frame=0
         channels [RingWidth {type=MovieSceneScriptingFloatChannel, keys [[0,0.666],[1,2],[15,0.666]]}]
        }]
      }]
    }
    HeatZoneIndicator_SmallState {
     id=#24
     possessed_class=Image
     tracks [{
       type=MovieSceneWidgetMaterialTrack
       sections [{
         type=MovieSceneParameterSection
         end_frame=15
         start_frame=0
         channels [RingWidth {type=MovieSceneScriptingFloatChannel, keys [[0,0.666],[1,2],[15,0.666]]}]
        }]
      }]
    }
    Backing_HeatZone {
     id=#11
     possessed_class=Image
     tracks [{
       type=MovieSceneWidgetMaterialTrack
       sections [{type=MovieSceneParameterSection, channels [RingWidth {type=MovieSceneScriptingFloatChannel, keys [[0,3],[1,9],[15,3]]}]}]
      }]
    }
    Progress_InThreshold {
     id=#12
     possessed_class=Image
     tracks [{
       type=MovieSceneWidgetMaterialTrack
       sections [{type=MovieSceneParameterSection, channels [RingWidth {type=MovieSceneScriptingFloatChannel, keys [[0,2.333],[1,9],[15,2.333]]}]}]
      }]
    }
   ]
   display_rate {numerator=60}
  }
 }
 LightningFXonFired {
  binding_count=3
  length_seconds=0.75
  object=&self:LightningFXonFired
  sequence {bindings [
    Lightning2 {
     id=#25
     possessed_class=Image
     tracks [{
       type=MovieSceneFloatTrack
       display_name=Render Opacity
       sections [{type=MovieSceneFloatSection, channels [None {type=MovieSceneScriptingFloatChannel, keys [[2.0003,0],[3.0003,0.25],[11,0]]}]}]
      }]
    }
    Lightning3 {
     id=#26
     possessed_class=Image
     tracks [{
       type=MovieSceneFloatTrack
       display_name=Render Opacity
       sections [{type=MovieSceneFloatSection, channels [None {type=MovieSceneScriptingFloatChannel, keys [[4,0],[5,0.5],[15,0]]}]}]
      }]
    }
    Lightning1 {
     id=#27
     possessed_class=Image
     tracks [{
       type=MovieSceneFloatTrack
       display_name=Render Opacity
       sections [{type=MovieSceneFloatSection, channels [None {type=MovieSceneScriptingFloatChannel, keys [[0,0],[1,0.125],[8,0]]}]}]
      }]
    }
   ]}
 }
 ShotsFired_MoreObvious_2 {
  binding_count=7
  length_seconds=0.25
  object=&self:ShotsFired_MoreObvious_2
  sequence {
   bindings [
    OverHeatedColorBorder {
     id=#20
     possessed_class=Border
     tracks [{
       type=MovieScene2DTransformTrack
       sections [{
         type=MovieScene2DTransformSection
         end_frame=15
         start_frame=0
         channels [
          Translation.X {type=MovieSceneScriptingFloatChannel, const [0,1]}
          Translation.Y {type=MovieSceneScriptingFloatChannel, const [0,1]}
          Angle {type=MovieSceneScriptingFloatChannel, const [0,1]}
          Scale.X {type=MovieSceneScriptingFloatChannel, keys [[1,1],[2,1.035],[15,1]]}
          Scale.Y {type=MovieSceneScriptingFloatChannel, keys [[1,1],[2,1.035],[15,1]]}
          Shear.X {type=MovieSceneScriptingFloatChannel, const [0,1]}
          Shear.Y {type=MovieSceneScriptingFloatChannel, const [0,1]}
         ]
        }]
      }]
    }
    HeatZoneIndicator {
     id=#21
     possessed_class=Image
     tracks [{
       type=MovieSceneWidgetMaterialTrack
       sections [{
         type=MovieSceneParameterSection
         end_frame=15
         start_frame=0
         channels [RingWidth {type=MovieSceneScriptingFloatChannel, keys [[0,0.666],[1,2],[15,0.666]]}]
        }]
      }]
    }
    OverlayHeatzone {
     id=#22
     possessed_class=Overlay
     tracks [{
       type=MovieScene2DTransformTrack
       sections [{
         type=MovieScene2DTransformSection
         end_frame=15
         start_frame=0
         channels [
          Translation.X {type=MovieSceneScriptingFloatChannel, keys [[0,0],[2,-1],[4,2],[5,-1],[9,1],[15,0]]}
          Translation.Y {type=MovieSceneScriptingFloatChannel, const [0,2]}
          Angle {type=MovieSceneScriptingFloatChannel, const [0,1]}
          Scale.X {type=MovieSceneScriptingFloatChannel, const [1,1]}
          Scale.Y {type=MovieSceneScriptingFloatChannel, const [1,1]}
          Shear.X {type=MovieSceneScriptingFloatChannel, const [0,1]}
          Shear.Y {type=MovieSceneScriptingFloatChannel, const [0,1]}
         ]
        }]
      }]
    }
    HeatZoneIndicator_Overheated {
     id=#23
     possessed_class=Image
     tracks [{
       type=MovieSceneWidgetMaterialTrack
       sections [{
         type=MovieSceneParameterSection
         end_frame=15
         start_frame=0
         channels [RingWidth {type=MovieSceneScriptingFloatChannel, keys [[0,0.666],[1,2],[15,0.666]]}]
        }]
      }]
    }
    HeatZoneIndicator_SmallState {
     id=#24
     possessed_class=Image
     tracks [{
       type=MovieSceneWidgetMaterialTrack
       sections [{
         type=MovieSceneParameterSection
         end_frame=15
         start_frame=0
         channels [RingWidth {type=MovieSceneScriptingFloatChannel, keys [[0,0.666],[1,2],[15,0.666]]}]
        }]
      }]
    }
    Backing_HeatZone {
     id=#11
     possessed_class=Image
     tracks [{
       type=MovieSceneWidgetMaterialTrack
       sections [{type=MovieSceneParameterSection, channels [RingWidth {type=MovieSceneScriptingFloatChannel, keys [[0,3],[1,8],[15,3]]}]}]
      }]
    }
    Progress_InThreshold {
     id=#12
     possessed_class=Image
     tracks [{
       type=MovieSceneWidgetMaterialTrack
       sections [{type=MovieSceneParameterSection, channels [RingWidth {type=MovieSceneScriptingFloatChannel, keys [[0,2.333],[1,8],[15,2.333]]}]}]
      }]
    }
   ]
   display_rate {numerator=60}
  }
 }
]
graphs [
 EventGraph {nodes [
   OnExitSuperheatEvent() {is=title K2Node_CustomEvent}
   > Set LastPowerupHeatState(LastPowerupHeatState:EPowerupHeatState Enum=Normal) -> (Output_Get:EPowerupHeatState Enum=Normal)
   > Set Num Shots In Superheated State(Num Shots In Superheated State:int=0) -> (Output_Get:int=0)
   > SetExitSuperHeatVisuals()
   OnEnterSuperheatEvent() {is=title K2Node_CustomEvent}
   > Set LastPowerupHeatState(LastPowerupHeatState=Superheated) -> (Output_Get=Normal)
   > Set Num Shots In Superheated State(Num Shots In Superheated State=0) -> (Output_Get=0)
   > SetEnterSuperHeatVisuals()
   OnEnterOverheatEvent() {is=title K2Node_CustomEvent}
   > Set LastPowerupHeatState(LastPowerupHeatState=Overheated) -> (Output_Get=Normal)
   > SetOverheatedVisuals!(execute, self, HasOverheated:bool=true) -> (then)
   OnExitOverheatEvent() {is=title K2Node_CustomEvent}
   > SetOverheatedVisuals(HasOverheated=false)
   n1 = Event PreConstruct()
   > Branch(Condition=@n1.IsDesignTime)
   Event Construct()
   n2 = GetEquippedWeapon!(self:Fort Weapon Reticle Extension Widget Base Object Reference) -> (ReturnValue:Fort Weapon Object Reference)
   > n3 = Cast To FortWeaponRanged!(execute, Object=@n2) -> (then, CastFailed, AsFort Weapon Ranged:Fort Weapon Ranged Object Reference) {is=title K2Node_DynamicCast}
   n4 = Create Event()
   > Bind Event to On Overheat Value Changed(self:Fort Weapon Ranged Object Reference=@n3, Delegate=@n4) {is=title K2Node_AddDelegate}
   > InitUpdateOverheatingState(self:Fort Weapon Ranged Object Reference=@n3)
   Event Destruct()
   > n5 = Cast To FortWeaponRanged(Object=@n2)
   > Unbind Event from On Overheat Value Changed(self:Fort Weapon Ranged Object Reference=@n5, Delegate=@n4) {is=title K2Node_RemoveDelegate}
   Event OnWeaponSingleFire() {is=title K2Node_Event}
   > Parent: OnWeaponSingleFire() {is=title K2Node_CallParentFunction}
   > Branch(Condition="Equal (Enum)"(
     A:EPowerupHeatState Enum=Get LastPowerupHeatState() -> (LastPowerupHeatState:EPowerupHeatState Enum=Normal)
     B:EPowerupHeatState Enum=Superheated
    ))
   > OnWeaponFiredWhileSuperheated()
   Event OnFullyExitVehicle() {is=title K2Node_Event}
   > n6 = Cast To FortWeaponRanged(Object=GetEquippedWeapon())
   > InitUpdateOverheatingState(self=@n6)
   UpdateOverheatBar!(execute, self, Current:double=0, Max:double=100, SuperHeatThreshold:double=25) -> (then)
   Get TestCurrentValue() -> (TestCurrentValue:double)
  ]}
 OnWeaponFiredWhileSuperheated {nodes [
   OnWeaponFiredWhileSuperheated() {is=title K2Node_FunctionEntry}
   > Increment Int(Value=Get Num Shots In Superheated State() -> (Num Shots In Superheated State:int=0))
   > PlayAnimation(InAnimation=Get ShotsFired_MoreObvious_2() -> (ShotsFired_MoreObvious_2:Widget Animation Object Reference))
   > PlayAnimation(InAnimation=Get LightningFXonFired() -> (LightningFXonFired:Widget Animation Object Reference))
   Get SupershotResponse() -> (SupershotResponse:Widget Animation Object Reference)
   Get ShotsFired() -> (ShotsFired:Widget Animation Object Reference)
   Get ShotsFired_MoreObvious() -> (ShotsFired_MoreObvious:Widget Animation Object Reference)
   Get BoostIndication() -> (BoostIndication:Widget Animation Object Reference)
  ]}
 OnWeaponOverheatValueChanged {nodes [
   n1 = OnWeaponOverheatValueChanged!() -> (then, CurrentValue:float=~, MaxValue:float=~, State:EFortWeaponOverheatState Enum=None, bIsInCooldown:bool=~) {is=title K2Node_FunctionEntry}
   > n2 = Cast To FortWeaponRanged(Object=GetEquippedWeapon())
   > UpdateOverheatBar(
    Current=@n1
    Max=@n1.MaxValue
    SuperHeatThreshold=GetValueAtLevel!(Input:"Scalable Float Structure (by ref)"=Get SuperheatTriggerPercent() -> (SuperheatTriggerPercent:Scalable Float Structure), Level:float=0) -> (ReturnValue:float)
   )
   > n3 = Branch(Condition=AND Boolean(
     A=IsOverheated!(self:Fort Weapon Ranged Object Reference=@n2) -> (ReturnValue:bool)
     B="float > float"(A=@n1, B=0) {is=title K2Node_CallFunction}
    ))
   n4 = "Equal (Enum)"(A=Overheated, B=Get LastPowerupHeatState() -> (LastPowerupHeatState=Normal))
   > n5 = Branch(Condition=@n4)
   OnEnterOverheatEvent(execute=@n5.else) {is=title K2Node_CallFunction}
   n6 = Branch(execute=@n3.else, Condition=@n4)
   > n7 = OnExitOverheatEvent() {is=title K2Node_CallFunction}
   n8 = Get SuperheatTriggerPercent()
   n9 = Branch(
    execute=[@n7.then, @n6.else]
    Condition="float >= float"(A=SafeDivide(A=@n1, B=@n1.MaxValue), B=GetValueAtLevel(Input=@n8, Level=0)) {is=title K2Node_CallFunction}
   )
   > n10 = Branch(Condition="Equal (Enum)"(A=Superheated, B=Get LastPowerupHeatState() -> (LastPowerupHeatState=Normal)))
   OnEnterSuperheatEvent(execute=@n10.else) {is=title K2Node_CallFunction}
   n11 = Branch(execute=@n9.else, Condition="Equal (Enum)"(A=Normal, B=Get LastPowerupHeatState() -> (LastPowerupHeatState=Normal)))
   OnExitSuperheatEvent(execute=@n11.else) {is=title K2Node_CallFunction}
  ]}
 Reset {nodes [
   Reset() {is=title K2Node_FunctionEntry}
   > SetOverheatedVisuals(HasOverheated=false)
   n1 = Get OnThresholdReached_NoRotation() -> (OnThresholdReached_NoRotation:Widget Animation Object Reference)
   > PlayAnimation(InAnimation=@n1, StartAtTime=GetEndTime(self=@n1), PlayMode=Reverse)
  ]}
 SetEnterSuperHeatVisuals {nodes [
   SetEnterSuperHeatVisuals() {is=title K2Node_FunctionEntry}
   > PlayAnimationForward(InAnimation=Get OnThresholdReached_NoRotation())
   > SetVisibility(
    self=[
     Get Glow() -> (Glow:Image Object Reference)
     Get LightningFX() -> (LightningFX:Overlay Object Reference)
     Get Glow_2() -> (Glow_2:Image Object Reference)
    ]
    InVisibility=SelfHitTestInvisible
   )
   > PlayAnimation(InAnimation=Get LightningFXonFired())
   > SetActiveWidgetIndex(self=Get IconSwitcher() -> (IconSwitcher:Widget Switcher Object Reference), Index=1)
   Get BoostIndicator() -> (BoostIndicator:Overlay Object Reference)
  ]}
 SetExitSuperHeatVisuals {nodes [
   SetExitSuperHeatVisuals() {is=title K2Node_FunctionEntry}
   > PlayAnimationReverse(InAnimation=Get OnThresholdReached_NoRotation())
   > SetVisibility(self=[Get Glow(), Get LightningFX(), Get Glow_2()], InVisibility=Collapsed)
   Get BoostIndicator()
  ]}
 SetOverheatedVisuals {nodes [
   n1 = SetOverheatedVisuals!() -> (then, HasOverheated:bool=~) {is=title K2Node_FunctionEntry}
   > n2 = Branch(Condition=@n1)
   > SetVisibility(
    self=[Get Backing_Overheated() -> (Backing_Overheated:Image Object Reference), Get WarningIcon() -> (WarningIcon:Overlay Object Reference)]
    InVisibility=SelfHitTestInvisible
   )
   > PlayAnimation(InAnimation=Get Overheated() -> (Overheated:Widget Animation Object Reference), NumLoopsToPlay=0)
   > SetContentColorAndOpacity!(
    execute
    self:Border Object Reference=Get OverHeatedColorBorder() -> (OverHeatedColorBorder:Border Object Reference)
    InContentColorAndOpacity:Linear Color Structure=(R=1,G=0.1046,B=0.0742,A=1)
   ) -> (then)
   > SetRenderOpacity(self=Get Progress_BelowThreshold() -> (Progress_BelowThreshold:Image Object Reference), InOpacity=0.25)
   > SetRenderOpacity(self=Get Backing() -> (Backing:Image Object Reference), InOpacity=0.5)
   > SetActiveWidgetIndex(self=Get IconSwitcher(), Index=0)
   SetVisibility(execute=@n2.else, self=[Get Backing_Overheated(), Get WarningIcon()], InVisibility=Collapsed)
   > StopAnimation(InAnimation=Get Overheated())
   > SetContentColorAndOpacity(self=Get OverHeatedColorBorder(), InContentColorAndOpacity=(R=1,G=1,B=1,A=1))
   > SetRenderOpacity(self=Get Progress_BelowThreshold(), InOpacity=1)
   > SetRenderOpacity(self=Get Backing(), InOpacity=1)
  ]}
 UpdateOverheatBar {nodes [
   n1 = UpdateOverheatBar!() -> (then, Current:double=~, Max:double=~, SuperHeatThreshold:double=~) {is=title K2Node_FunctionEntry}
   > SetRenderOpacity(self=Get LightningFX(), InOpacity=@n1)
   > n2 = GetDynamicMaterial(self=Get Progress_BelowThreshold())
   n3 = float / float(A=@n1, B=@n1.Max) {is=title K2Node_CallFunction}
   n4 = float * float(A=@n3, B=100)
   > SetScalarParameterValue(self=@n2, ParameterName=FillPercent, Value=@n4)
   > n5 = GetDynamicMaterial(self=Get Progress_InThreshold() -> (Progress_InThreshold:Image Object Reference))
   > SetScalarParameterValue(self=@n5, ParameterName=FillPercent, Value=@n4)
   > n6 = GetDynamicMaterial(self=Get Backing())
   > SetScalarParameterValue(self=@n6, ParameterName=FillPercent, Value=float * float(A=Lerp(A=1, B=0, Alpha=@n3), B=100))
   > n7 = GetDynamicMaterial(self=Get HeatZoneIndicator_TopTick_Overheated() -> (HeatZoneIndicator_TopTick_Overheated:Image Object Reference))
   > SetScalarParameterValue(self=@n7, ParameterName=AnimateOverallAlpha, Value=@n3)
   > n8 = GetDynamicMaterial(self=Get HeatZoneIndicator_Overheated() -> (HeatZoneIndicator_Overheated:Image Object Reference))
   > SetScalarParameterValue(self=@n8, ParameterName=Overall_Alpha, Value=int * float!(A:int=6, B:double=Lerp(A=-1, B=1, Alpha=@n3)) -> (ReturnValue:double))
   > n9 = GetDynamicMaterial(self=Get Glow())
   n10 = Lerp(A=0, B=0.5, Alpha=@n3)
   > SetScalarParameterValue(self=@n9, ParameterName=Final_Alpha, Value=@n10)
   > n11 = GetDynamicMaterial(self=Get Glow_2())
   > SetScalarParameterValue(self=@n11, ParameterName=Final_Alpha, Value=@n10)
   > n12 = GetDynamicMaterial(self=Get Backing_HeatZone() -> (Backing_HeatZone:Image Object Reference))
   > SetScalarParameterValue(self=@n12, ParameterName=Overall_Alpha, Value=Lerp(A=0.4, B=0.5, Alpha=@n3))
   > n13 = GetDynamicMaterial(self=Get Backing())
   > SetScalarParameterValue(self=@n13, ParameterName=Overall_Alpha, Value=Lerp(A=0.5, B=0.75, Alpha=@n3))
   > n14 = GetDynamicMaterial(self=Get Backing_HeatZone())
   n15 = float * float(A=Get SuperHeatThreshold() -> (SuperHeatThreshold:double), B=25)
   n16 = float - float(A=25, B=@n15) {is=title K2Node_CallFunction}
   > SetScalarParameterValue(self=@n14, ParameterName=Section_MaskPercent, Value=@n16)
   > n17 = GetDynamicMaterial(self=Get Backing_Overheated())
   > SetScalarParameterValue(self=@n17, ParameterName=Section_MaskPercent, Value=@n16)
   > n18 = GetDynamicMaterial(self=Get HeatZoneIndicator() -> (HeatZoneIndicator:Image Object Reference))
   > SetScalarParameterValue(self=@n18, ParameterName=Section_MaskPercent, Value=@n16)
   > n19 = GetDynamicMaterial(self=Get HeatZoneIndicator_SmallState() -> (HeatZoneIndicator_SmallState:Image Object Reference))
   > SetScalarParameterValue(self=@n19, ParameterName=Section_MaskPercent, Value=@n16)
   > n20 = GetDynamicMaterial(self=Get HeatZoneIndicator_Overheated())
   > SetScalarParameterValue(self=@n20, ParameterName=Section_MaskPercent, Value=@n16)
   > n21 = GetDynamicMaterial(self=Get Progress_InThreshold())
   > SetScalarParameterValue(self=@n21, ParameterName=Section_MaskPercent, Value=@n15)
   > n22 = GetDynamicMaterial(self=Get Progress_BelowThreshold())
   > SetScalarParameterValue(self=@n22, ParameterName=Section_MaskPercent, Value=@n15)
   > n23 = GetDynamicMaterial(self=Get HeatZoneIndicator_BottomTick() -> (HeatZoneIndicator_BottomTick:Image Object Reference))
   n24 = float * float(A=@n15, B=0.01)
   > SetScalarParameterValue(self=@n23, ParameterName=PostRotation, Value=@n24)
   > n25 = GetDynamicMaterial(self=Get HeatZoneIndicator_BottomTick_SmallState() -> (HeatZoneIndicator_BottomTick_SmallState:Image Object Reference))
   > SetScalarParameterValue(self=@n25, ParameterName=PostRotation, Value=@n24)
  ]}
]
variables [
 Current {type=number}
 HasOverheated {type=boolean}
 IndicatorOverheatBarColor {type=LinearColor, default=<rgba 1 0.6038 0.0497 1>}
 Is in Cooldown {type=boolean}
 NewVar_0 {type=number}
 Num Shots In Superheated State {type=integer}
 Powerup Gameplay Ability {type=GA_Athena_Powerup_HeatEvent_Parent_C, default=null}
 Super Heat Threshold {type=number}
 SuperheatTriggerPercent {type=ScalableFloat, default {registryType=None, value=0.25, curve {curveTable=null, rowName=None}}}
 TestCurrentValue {type=number, default=50}
]
widgets {
 info {
  inheritedWidgetCount=1
  parentClass=/Script/CorruptionGameplayCodeUI.FortPowerupReticleExtensionWidget
  rootWidgetClass=&Overlay
  widgetCount=44
 }
 widgets [
  Overlay_31 {widgetClassPath=&Overlay, widget=Overlay_31}
  SizeBox_0 {
   widgetClassPath=&SizeBox
   parent=Overlay_31
   widget=SizeBox_0
   properties {bOverride_HeightOverride=true, bOverride_WidthOverride=true, heightOverride=128, widthOverride=128}
   slot {class=OverlaySlot}
  }
  Overlay_Body {
   widgetClassPath=&Overlay
   parent=SizeBox_0
   widget=Overlay_Body
   properties {renderTransform {angle=0, scale=<1 1>, shear=<0 0>, translation=<10 10>}}
   slot {class=SizeBoxSlot}
  }
  OverlayContent {
   widgetClassPath=&Overlay
   parent=Overlay_Body
   widget=OverlayContent
   properties {renderOpacity=0.5}
   slot {class=OverlaySlot, properties {horizontalAlignment=HAlign_Fill, verticalAlignment=VAlign_Fill}}
  }
  BackingsMeters {
   widgetClassPath=&Overlay
   bIsVariable=true
   parent=OverlayContent
   widget=BackingsMeters
   slot {class=OverlaySlot, properties {horizontalAlignment=HAlign_Fill, verticalAlignment=VAlign_Fill}}
  }
  Backing {
   widgetClassPath=&Image
   bIsVariable=true
   parent=BackingsMeters
   widget=Backing
   properties {
    visibility=SelfHitTestInvisible
    brush {
     bIsDynamicallyLoaded=false
     drawAs=Image
     imageType=NoImage
     mirroring=NoMirror
     resourceName=None
     resourceObject=/CorruptionGameplay/UI/ReticleExtensions/Art/M_UI_PwrUp_Overheat_Clamp_Backing
     tiling=NoTile
     imageSize=<32 32>
     margin=<ltrb 0 0 0 0>
     outlineSettings {
      bUseBrushTransparency=false
      roundingType=HalfHeightRadius
      width=0
      color {colorUseRule=UseColor_Specified, specifiedColor=<rgba 0 0 0 0>}
      cornerRadii=<0 0 0 0>
     }
     tintColor {colorUseRule=UseColor_Specified, specifiedColor=<rgba 1 1 1 1>}
     uVRegion {bIsValid=false, max=<0 0>, min=<0 0>}
    }
   }
   slot {class=OverlaySlot, properties {horizontalAlignment=HAlign_Fill, verticalAlignment=VAlign_Fill}}
  }
  Progress_BelowThreshold {
   widgetClassPath=&Image
   bIsVariable=true
   parent=BackingsMeters
   widget=Progress_BelowThreshold
   properties {
    visibility=SelfHitTestInvisible
    brush {
     bIsDynamicallyLoaded=false
     drawAs=Image
     imageType=NoImage
     mirroring=NoMirror
     resourceName=None
     resourceObject=/CorruptionGameplay/UI/ReticleExtensions/Art/M_UI_PwrUp_Overheat_Filll
     tiling=NoTile
     imageSize=<128 128>
     margin=<ltrb 0 0 0 0>
     outlineSettings {
      bUseBrushTransparency=false
      roundingType=HalfHeightRadius
      width=0
      color {colorUseRule=UseColor_Specified, specifiedColor=<rgba 0 0 0 0>}
      cornerRadii=<0 0 0 0>
     }
     tintColor {colorUseRule=UseColor_Specified, specifiedColor=<rgba 1 1 1 1>}
     uVRegion {bIsValid=false, max=<0 0>, min=<0 0>}
    }
   }
   slot {class=OverlaySlot, properties {horizontalAlignment=HAlign_Fill, verticalAlignment=VAlign_Fill}}
  }
  Backing_HeatZone {
   widgetClassPath=&Image
   bIsVariable=true
   parent=BackingsMeters
   widget=Backing_HeatZone
   properties {
    visibility=SelfHitTestInvisible
    brush {
     bIsDynamicallyLoaded=false
     drawAs=Image
     imageType=NoImage
     mirroring=NoMirror
     resourceName=None
     resourceObject=/CorruptionGameplay/UI/ReticleExtensions/Art/M_UI_PwrUp_Overheat_Clamp_Backing_Heatzone
     tiling=NoTile
     imageSize=<32 32>
     margin=<ltrb 0 0 0 0>
     outlineSettings {
      bUseBrushTransparency=false
      roundingType=HalfHeightRadius
      width=0
      color {colorUseRule=UseColor_Specified, specifiedColor=<rgba 0 0 0 0>}
      cornerRadii=<0 0 0 0>
     }
     tintColor {colorUseRule=UseColor_Specified, specifiedColor=<rgba 1 1 1 1>}
     uVRegion {bIsValid=false, max=<0 0>, min=<0 0>}
    }
   }
   slot {class=OverlaySlot, properties {horizontalAlignment=HAlign_Fill, verticalAlignment=VAlign_Fill}}
  }
  Backing_Overheated {
   widgetClassPath=&Image
   bIsVariable=true
   parent=BackingsMeters
   widget=Backing_Overheated
   properties {
    visibility=Hidden
    brush {
     bIsDynamicallyLoaded=false
     drawAs=Image
     imageType=NoImage
     mirroring=NoMirror
     resourceName=None
     resourceObject=/CorruptionGameplay/UI/ReticleExtensions/Art/M_UI_PwrUp_OverheatedState
     tiling=NoTile
     imageSize=<32 32>
     margin=<ltrb 0 0 0 0>
     outlineSettings {
      bUseBrushTransparency=false
      roundingType=HalfHeightRadius
      width=0
      color {colorUseRule=UseColor_Specified, specifiedColor=<rgba 0 0 0 0>}
      cornerRadii=<0 0 0 0>
     }
     tintColor {colorUseRule=UseColor_Specified, specifiedColor=<rgba 1 1 1 1>}
     uVRegion {bIsValid=false, max=<0 0>, min=<0 0>}
    }
   }
   slot {class=OverlaySlot, properties {horizontalAlignment=HAlign_Fill, verticalAlignment=VAlign_Fill}}
  }
  Glow {
   widgetClassPath=&Image
   bIsVariable=true
   parent=BackingsMeters
   widget=Glow
   properties {
    visibility=Collapsed
    brush {
     bIsDynamicallyLoaded=false
     drawAs=Image
     imageType=NoImage
     mirroring=NoMirror
     resourceName=None
     resourceObject=/CorruptionGameplay/UI/ReticleExtensions/Art/M_UI_PwrUp_Circle_Glow
     tiling=NoTile
     imageSize=<32 32>
     margin=<ltrb 0 0 0 0>
     outlineSettings {
      bUseBrushTransparency=false
      roundingType=HalfHeightRadius
      width=0
      color {colorUseRule=UseColor_Specified, specifiedColor=<rgba 0 0 0 0>}
      cornerRadii=<0 0 0 0>
     }
     tintColor {colorUseRule=UseColor_Specified, specifiedColor=<rgba 1 1 1 1>}
     uVRegion {bIsValid=false, max=<0 0>, min=<0 0>}
    }
    renderTransform {angle=0, scale=<1.2 1.2>, shear=<0 0>, translation=<4 4>}
   }
   slot {class=OverlaySlot, properties {horizontalAlignment=HAlign_Fill, verticalAlignment=VAlign_Fill}}
  }
  Glow_2 {
   widgetClassPath=&Image
   bIsVariable=true
   parent=BackingsMeters
   widget=Glow_2
   properties {
    visibility=Collapsed
    brush {
     bIsDynamicallyLoaded=false
     drawAs=Image
     imageType=NoImage
     mirroring=NoMirror
     resourceName=None
     resourceObject=/CorruptionGameplay/UI/ReticleExtensions/Art/M_UI_PwrUp_Circle_Glow_Inst
     tiling=NoTile
     imageSize=<32 32>
     margin=<ltrb 0 0 0 0>
     outlineSettings {
      bUseBrushTransparency=false
      roundingType=HalfHeightRadius
      width=0
      color {colorUseRule=UseColor_Specified, specifiedColor=<rgba 0 0 0 0>}
      cornerRadii=<0 0 0 0>
     }
     tintColor {colorUseRule=UseColor_Specified, specifiedColor=<rgba 1 1 1 1>}
     uVRegion {bIsValid=false, max=<0 0>, min=<0 0>}
    }
    renderTransform {angle=0, scale=<1.2 1.2>, shear=<0 0>, translation=<4 4>}
   }
   slot {class=OverlaySlot, properties {horizontalAlignment=HAlign_Fill, verticalAlignment=VAlign_Fill}}
  }
  Progress_InThreshold {
   widgetClassPath=&Image
   bIsVariable=true
   parent=BackingsMeters
   widget=Progress_InThreshold
   properties {
    visibility=SelfHitTestInvisible
    brush {
     bIsDynamicallyLoaded=false
     drawAs=Image
     imageType=NoImage
     mirroring=NoMirror
     resourceName=None
     resourceObject=/CorruptionGameplay/UI/ReticleExtensions/Art/M_UI_PwrUp_SuperHeatedPortion
     tiling=NoTile
     imageSize=<128 128>
     margin=<ltrb 0 0 0 0>
     outlineSettings {
      bUseBrushTransparency=false
      roundingType=HalfHeightRadius
      width=0
      color {colorUseRule=UseColor_Specified, specifiedColor=<rgba 0 0 0 0>}
      cornerRadii=<0 0 0 0>
     }
     tintColor {colorUseRule=UseColor_Specified, specifiedColor=<rgba 1 1 1 1>}
     uVRegion {bIsValid=false, max=<0 0>, min=<0 0>}
    }
   }
   slot {class=OverlaySlot, properties {horizontalAlignment=HAlign_Fill, verticalAlignment=VAlign_Fill}}
  }
  OverHeatedColorBorder {
   widgetClassPath=&Border
   bIsVariable=true
   parent=OverlayContent
   widget=OverHeatedColorBorder
   properties {
    background {
     bIsDynamicallyLoaded=false
     drawAs=NoDrawType
     imageType=NoImage
     mirroring=NoMirror
     resourceName=None
     resourceObject=null
     tiling=NoTile
     imageSize=<32 32>
     margin=<ltrb 0 0 0 0>
     outlineSettings {
      bUseBrushTransparency=false
      roundingType=HalfHeightRadius
      width=0
      color {colorUseRule=UseColor_Specified, specifiedColor=<rgba 0 0 0 0>}
      cornerRadii=<0 0 0 0>
     }
     tintColor {colorUseRule=UseColor_Specified, specifiedColor=<rgba 1 1 1 1>}
     uVRegion {bIsValid=false, max=<0 0>, min=<0 0>}
    }
    padding=<ltrb 0 0 0 0>
   }
   slot {class=OverlaySlot, properties {horizontalAlignment=HAlign_Fill, verticalAlignment=VAlign_Fill}}
  }
  OverlayHeatzone {
   widgetClassPath=&Overlay
   parent=OverHeatedColorBorder
   widget=OverlayHeatzone
   slot {class=BorderSlot, properties {padding=<ltrb 0 0 0 0>}}
  }
  HeatZoneIndicator_TopTick_SmallState {
   widgetClassPath=&Image
   bIsVariable=true
   parent=OverlayHeatzone
   widget=HeatZoneIndicator_TopTick_SmallState
   properties {
    visibility=SelfHitTestInvisible
    brush {
     bIsDynamicallyLoaded=false
     drawAs=Image
     imageType=NoImage
     mirroring=NoMirror
     resourceName=None
     resourceObject=/CorruptionGameplay/UI/ReticleExtensions/Art/M_UI_PwrUp_Notch_Top
     tiling=NoTile
     imageSize=<32 32>
     margin=<ltrb 0 0 0 0>
     outlineSettings {
      bUseBrushTransparency=false
      roundingType=HalfHeightRadius
      width=0
      color {colorUseRule=UseColor_Specified, specifiedColor=<rgba 0 0 0 0>}
      cornerRadii=<0 0 0 0>
     }
     tintColor {colorUseRule=UseColor_Specified, specifiedColor=<rgba 1 1 1 1>}
     uVRegion {bIsValid=false, max=<0 0>, min=<0 0>}
    }
   }
   slot {class=OverlaySlot, properties {horizontalAlignment=HAlign_Fill, verticalAlignment=VAlign_Fill}}
  }
  HeatZoneIndicator_TopTick {
   widgetClassPath=&Image
   bIsVariable=true
   parent=OverlayHeatzone
   widget=HeatZoneIndicator_TopTick
   properties {
    visibility=SelfHitTestInvisible
    brush {
     bIsDynamicallyLoaded=false
     drawAs=Image
     imageType=NoImage
     mirroring=NoMirror
     resourceName=None
     resourceObject=&M_UI_PwrUp_Notch_Top
     tiling=NoTile
     imageSize=<32 32>
     margin=<ltrb 0 0 0 0>
     outlineSettings {
      bUseBrushTransparency=false
      roundingType=HalfHeightRadius
      width=0
      color {colorUseRule=UseColor_Specified, specifiedColor=<rgba 0 0 0 0>}
      cornerRadii=<0 0 0 0>
     }
     tintColor {colorUseRule=UseColor_Specified, specifiedColor=<rgba 1 1 1 1>}
     uVRegion {bIsValid=false, max=<0 0>, min=<0 0>}
    }
   }
   slot {class=OverlaySlot, properties {horizontalAlignment=HAlign_Fill, verticalAlignment=VAlign_Fill}}
  }
  HeatZoneIndicator_TopTick_Overheated {
   widgetClassPath=&Image
   bIsVariable=true
   parent=OverlayHeatzone
   widget=HeatZoneIndicator_TopTick_Overheated
   properties {
    visibility=SelfHitTestInvisible
    brush {
     bIsDynamicallyLoaded=false
     drawAs=Image
     imageType=NoImage
     mirroring=NoMirror
     resourceName=None
     resourceObject=/CorruptionGameplay/UI/ReticleExtensions/Art/M_UI_PwrUp_Notch_Top_Overheated
     tiling=NoTile
     imageSize=<32 32>
     margin=<ltrb 0 0 0 0>
     outlineSettings {
      bUseBrushTransparency=false
      roundingType=HalfHeightRadius
      width=0
      color {colorUseRule=UseColor_Specified, specifiedColor=<rgba 0 0 0 0>}
      cornerRadii=<0 0 0 0>
     }
     tintColor {colorUseRule=UseColor_Specified, specifiedColor=<rgba 1 1 1 1>}
     uVRegion {bIsValid=false, max=<0 0>, min=<0 0>}
    }
   }
   slot {class=OverlaySlot, properties {horizontalAlignment=HAlign_Fill, verticalAlignment=VAlign_Fill}}
  }
  HeatZoneIndicator_BottomTick_SmallState {
   widgetClassPath=&Image
   bIsVariable=true
   parent=OverlayHeatzone
   widget=HeatZoneIndicator_BottomTick_SmallState
   properties {
    visibility=SelfHitTestInvisible
    brush {
     bIsDynamicallyLoaded=false
     drawAs=Image
     imageType=NoImage
     mirroring=NoMirror
     resourceName=None
     resourceObject=/CorruptionGameplay/UI/ReticleExtensions/Art/M_UI_PwrUp_Notch_Bottom
     tiling=NoTile
     imageSize=<32 32>
     margin=<ltrb 0 0 0 0>
     outlineSettings {
      bUseBrushTransparency=false
      roundingType=HalfHeightRadius
      width=0
      color {colorUseRule=UseColor_Specified, specifiedColor=<rgba 0 0 0 0>}
      cornerRadii=<0 0 0 0>
     }
     tintColor {colorUseRule=UseColor_Specified, specifiedColor=<rgba 1 1 1 1>}
     uVRegion {bIsValid=false, max=<0 0>, min=<0 0>}
    }
   }
   slot {class=OverlaySlot, properties {horizontalAlignment=HAlign_Fill, verticalAlignment=VAlign_Fill}}
  }
  HeatZoneIndicator_BottomTick {
   widgetClassPath=&Image
   bIsVariable=true
   parent=OverlayHeatzone
   widget=HeatZoneIndicator_BottomTick
   properties {
    visibility=SelfHitTestInvisible
    brush {
     bIsDynamicallyLoaded=false
     drawAs=Image
     imageType=NoImage
     mirroring=NoMirror
     resourceName=None
     resourceObject=&M_UI_PwrUp_Notch_Bottom
     tiling=NoTile
     imageSize=<32 32>
     margin=<ltrb 0 0 0 0>
     outlineSettings {
      bUseBrushTransparency=false
      roundingType=HalfHeightRadius
      width=0
      color {colorUseRule=UseColor_Specified, specifiedColor=<rgba 0 0 0 0>}
      cornerRadii=<0 0 0 0>
     }
     tintColor {colorUseRule=UseColor_Specified, specifiedColor=<rgba 1 1 1 1>}
     uVRegion {bIsValid=false, max=<0 0>, min=<0 0>}
    }
   }
   slot {class=OverlaySlot, properties {horizontalAlignment=HAlign_Fill, verticalAlignment=VAlign_Fill}}
  }
  HeatZoneIndicator_SmallState {
   widgetClassPath=&Image
   bIsVariable=true
   parent=OverlayHeatzone
   widget=HeatZoneIndicator_SmallState
   properties {
    visibility=SelfHitTestInvisible
    brush {
     bIsDynamicallyLoaded=false
     drawAs=Image
     imageType=NoImage
     mirroring=NoMirror
     resourceName=None
     resourceObject=/CorruptionGameplay/UI/ReticleExtensions/Art/M_UI_PwrUp_Overheat_Clamp_Backing_Heatzone_Indicator_Sm
     tiling=NoTile
     imageSize=<32 32>
     margin=<ltrb 0 0 0 0>
     outlineSettings {
      bUseBrushTransparency=false
      roundingType=HalfHeightRadius
      width=0
      color {colorUseRule=UseColor_Specified, specifiedColor=<rgba 0 0 0 0>}
      cornerRadii=<0 0 0 0>
     }
     tintColor {colorUseRule=UseColor_Specified, specifiedColor=<rgba 1 1 1 1>}
     uVRegion {bIsValid=false, max=<0 0>, min=<0 0>}
    }
   }
   slot {class=OverlaySlot, properties {horizontalAlignment=HAlign_Fill, verticalAlignment=VAlign_Fill}}
  }
  HeatZoneIndicator {
   widgetClassPath=&Image
   bIsVariable=true
   parent=OverlayHeatzone
   widget=HeatZoneIndicator
   properties {
    visibility=SelfHitTestInvisible
    brush {
     bIsDynamicallyLoaded=false
     drawAs=Image
     imageType=NoImage
     mirroring=NoMirror
     resourceName=None
     resourceObject=/CorruptionGameplay/UI/ReticleExtensions/Art/M_UI_PwrUp_Overheat_Clamp_Backing_Heatzone_Indicator
     tiling=NoTile
     imageSize=<32 32>
     margin=<ltrb 0 0 0 0>
     outlineSettings {
      bUseBrushTransparency=false
      roundingType=HalfHeightRadius
      width=0
      color {colorUseRule=UseColor_Specified, specifiedColor=<rgba 0 0 0 0>}
      cornerRadii=<0 0 0 0>
     }
     tintColor {colorUseRule=UseColor_Specified, specifiedColor=<rgba 1 1 1 1>}
     uVRegion {bIsValid=false, max=<0 0>, min=<0 0>}
    }
   }
   slot {class=OverlaySlot, properties {horizontalAlignment=HAlign_Fill, verticalAlignment=VAlign_Fill}}
  }
  HeatZoneIndicator_Overheated {
   widgetClassPath=&Image
   bIsVariable=true
   parent=OverlayHeatzone
   widget=HeatZoneIndicator_Overheated
   properties {
    visibility=SelfHitTestInvisible
    brush {
     bIsDynamicallyLoaded=false
     drawAs=Image
     imageType=NoImage
     mirroring=NoMirror
     resourceName=None
     resourceObject=/CorruptionGameplay/UI/ReticleExtensions/Art/M_UI_PwrUp_Overheat_Clamp_Overheated
     tiling=NoTile
     imageSize=<32 32>
     margin=<ltrb 0 0 0 0>
     outlineSettings {
      bUseBrushTransparency=false
      roundingType=HalfHeightRadius
      width=0
      color {colorUseRule=UseColor_Specified, specifiedColor=<rgba 0 0 0 0>}
      cornerRadii=<0 0 0 0>
     }
     tintColor {colorUseRule=UseColor_Specified, specifiedColor=<rgba 1 1 1 1>}
     uVRegion {bIsValid=false, max=<0 0>, min=<0 0>}
    }
   }
   slot {class=OverlaySlot, properties {horizontalAlignment=HAlign_Fill, verticalAlignment=VAlign_Fill}}
  }
  IconSwitcher {
   widgetClassPath=&WidgetSwitcher
   bIsVariable=true
   parent=Overlay_Body
   widget=IconSwitcher
   properties {activeWidgetIndex=1, visibility=Collapsed}
   slot {class=OverlaySlot, properties {horizontalAlignment=HAlign_Right, verticalAlignment=VAlign_Center}}
  }
  WarningIcon {
   widgetClassPath=&Overlay
   bIsVariable=true
   parent=IconSwitcher
   widget=WarningIcon
   properties {visibility=Collapsed, renderTransform {angle=0, scale=<0.8 0.8>, shear=<0 0>, translation=<9 -13>}}
   slot {class=WidgetSwitcherSlot}
  }
  BatteryIconWarningGlow {
   widgetClassPath=&Image
   bIsVariable=true
   parent=WarningIcon
   widget=BatteryIconWarningGlow
   properties {
    visibility=SelfHitTestInvisible
    brush {
     bIsDynamicallyLoaded=false
     drawAs=Image
     imageType=NoImage
     mirroring=NoMirror
     resourceName=None
     resourceObject=/CorruptionGameplay/UI/ReticleExtensions/Art/M_UI_PwrUp_Glow
     tiling=NoTile
     imageSize=<24 24>
     margin=<ltrb 0.5 0.5 0.5 0.5>
     outlineSettings {
      bUseBrushTransparency=false
      roundingType=HalfHeightRadius
      width=0
      color {colorUseRule=UseColor_Specified, specifiedColor=<rgba 0 0 0 0>}
      cornerRadii=<0 0 0 0>
     }
     tintColor {colorUseRule=UseColor_Specified, specifiedColor=<rgba 1 1 1 1>}
     uVRegion {bIsValid=false, max=<0 0>, min=<0 0>}
    }
    renderTransform {angle=0, scale=<2.5 2.5>, shear=<0 0>, translation=<0 0>}
   }
   slot {class=OverlaySlot, properties {horizontalAlignment=HAlign_Center, verticalAlignment=VAlign_Center}}
  }
  Image_1 {
   widgetClassPath=&Image
   bIsVariable=true
   parent=WarningIcon
   widget=Image_1
   properties {
    visibility=SelfHitTestInvisible
    brush {
     bIsDynamicallyLoaded=false
     drawAs=Image
     imageType=NoImage
     mirroring=NoMirror
     resourceName=None
     resourceObject=/CorruptionGameplay/UI/ReticleExtensions/Art/M_UI_PwrUp_Circle_Warning_Ring
     tiling=NoTile
     imageSize=<20 20>
     margin=<ltrb 0 0 0 0>
     outlineSettings {
      bUseBrushTransparency=false
      roundingType=HalfHeightRadius
      width=0
      color {colorUseRule=UseColor_Specified, specifiedColor=<rgba 0 0 0 0>}
      cornerRadii=<0 0 0 0>
     }
     tintColor {colorUseRule=UseColor_Specified, specifiedColor=<rgba 1 1 1 1>}
     uVRegion {bIsValid=false, max=<0 0>, min=<0 0>}
    }
   }
   slot {class=OverlaySlot, properties {horizontalAlignment=HAlign_Center, verticalAlignment=VAlign_Center}}
  }
  Image_2 {
   widgetClassPath=&Image
   bIsVariable=true
   parent=WarningIcon
   widget=Image_2
   properties {
    visibility=SelfHitTestInvisible
    brush {
     bIsDynamicallyLoaded=false
     drawAs=Image
     imageType=NoImage
     mirroring=NoMirror
     resourceName=None
     resourceObject=/CorruptionGameplay/UI/ReticleExtensions/Art/M_UI_PwrUp_Circle_Warning
     tiling=NoTile
     imageSize=<20 20>
     margin=<ltrb 0 0 0 0>
     outlineSettings {
      bUseBrushTransparency=false
      roundingType=HalfHeightRadius
      width=0
      color {colorUseRule=UseColor_Specified, specifiedColor=<rgba 0 0 0 0>}
      cornerRadii=<0 0 0 0>
     }
     tintColor {colorUseRule=UseColor_Specified, specifiedColor=<rgba 1 1 1 1>}
     uVRegion {bIsValid=false, max=<0 0>, min=<0 0>}
    }
   }
   slot {class=OverlaySlot, properties {horizontalAlignment=HAlign_Center, verticalAlignment=VAlign_Center}}
  }
  CommonTextBlock {
   widgetClassPath=&CommonTextBlock
   parent=WarningIcon
   widget=CommonTextBlock
   properties {
    justification=Center
    renderOpacity=0.5
    text="!"
    colorAndOpacity {colorUseRule=UseColor_Specified, specifiedColor=<rgba 0.0156 0 0.0625 1>}
    font {
     bForceMonospaced=false
     bMaterialIsStencil=false
     fontMaterial=null
     fontObject=&BurbankSmall
     letterSpacing=0
     monospacedWidth=1
     size=10
     skewAmount=0
     typefaceFontName=Black
     outlineSettings {
      bApplyOutlineToDropShadows=false
      bMiteredCorners=false
      bSeparateFillAlpha=false
      outlineMaterial=null
      outlineSize=0
      outlineColor=<rgba 0 0 0 1>
     }
    }
    renderTransform {angle=4, scale=<1 1>, shear=<0 0>, translation=<0 0>}
    shadowOffset=<0 0>
   }
   slot {class=OverlaySlot, properties {horizontalAlignment=HAlign_Center, verticalAlignment=VAlign_Center}}
  }
  Image_91 {
   widgetClassPath=&Image
   bIsVariable=true
   parent=WarningIcon
   widget=Image_91
   properties {
    visibility=Hidden
    brush {
     bIsDynamicallyLoaded=false
     drawAs=Image
     imageType=NoImage
     mirroring=NoMirror
     resourceName=None
     resourceObject=/CorruptionGameplay/UI/ReticleExtensions/Art/M_UI_PwrUp_BoostIcon_master_Inst
     tiling=NoTile
     imageSize=<12 12>
     margin=<ltrb 0 0 0 0>
     outlineSettings {
      bUseBrushTransparency=false
      roundingType=HalfHeightRadius
      width=0
      color {colorUseRule=UseColor_Specified, specifiedColor=<rgba 0 0 0 0>}
      cornerRadii=<0 0 0 0>
     }
     tintColor {colorUseRule=UseColor_Specified, specifiedColor=<rgba 1 1 1 1>}
     uVRegion {bIsValid=false, max=<0 0>, min=<0 0>}
    }
   }
   slot {class=OverlaySlot, properties {horizontalAlignment=HAlign_Center, verticalAlignment=VAlign_Center}}
  }
  BoostIndicator {
   widgetClassPath=&Overlay
   bIsVariable=true
   parent=IconSwitcher
   widget=BoostIndicator
   properties {renderOpacity=0.5, visibility=Collapsed, renderTransform {angle=0, scale=<0.75 0.75>, shear=<0 0>, translation=<13 -13>}}
   slot {class=WidgetSwitcherSlot, properties {horizontalAlignment=HAlign_Right, verticalAlignment=VAlign_Center}}
  }
  Lightning {
   widgetClassPath=&Image
   bIsVariable=true
   parent=BoostIndicator
   widget=Lightning
   properties {
    renderOpacity=0
    visibility=SelfHitTestInvisible
    brush {
     bIsDynamicallyLoaded=false
     drawAs=Image
     imageType=NoImage
     mirroring=NoMirror
     resourceName=None
     resourceObject=/Game/Athena/HUD/XP/Materials/M_UI_PwrUp_Lightning
     tiling=NoTile
     imageSize=<32 32>
     margin=<ltrb 0 0 0 0>
     outlineSettings {
      bUseBrushTransparency=false
      roundingType=HalfHeightRadius
      width=0
      color {colorUseRule=UseColor_Specified, specifiedColor=<rgba 0 0 0 0>}
      cornerRadii=<0 0 0 0>
     }
     tintColor {colorUseRule=UseColor_Specified, specifiedColor=<rgba 1 1 1 1>}
     uVRegion {bIsValid=false, max=<0 0>, min=<0 0>}
    }
    renderTransform {angle=0, scale=<2.5 2.5>, shear=<0 0>, translation=<0 0>}
   }
   slot {class=OverlaySlot}
  }
  Noise {
   widgetClassPath=&Image
   bIsVariable=true
   parent=BoostIndicator
   widget=Noise
   properties {
    renderOpacity=0
    visibility=SelfHitTestInvisible
    brush {
     bIsDynamicallyLoaded=false
     drawAs=Image
     imageType=NoImage
     mirroring=NoMirror
     resourceName=None
     resourceObject=/CorruptionGameplay/UI/ReticleExtensions/Art/M_UI_PwrUP_Flare
     tiling=NoTile
     imageSize=<32 32>
     margin=<ltrb 0 0 0 0>
     outlineSettings {
      bUseBrushTransparency=false
      roundingType=HalfHeightRadius
      width=0
      color {colorUseRule=UseColor_Specified, specifiedColor=<rgba 0 0 0 0>}
      cornerRadii=<0 0 0 0>
     }
     tintColor {colorUseRule=UseColor_Specified, specifiedColor=<rgba 1 1 1 1>}
     uVRegion {bIsValid=false, max=<0 0>, min=<0 0>}
    }
    renderTransform {angle=0, scale=<2 2>, shear=<0 0>, translation=<0 0>}
   }
   slot {class=OverlaySlot}
  }
  BatteryIconWarningGlow_1 {
   widgetClassPath=&Image
   bIsVariable=true
   parent=BoostIndicator
   widget=BatteryIconWarningGlow_1
   properties {
    visibility=SelfHitTestInvisible
    brush {
     bIsDynamicallyLoaded=false
     drawAs=Image
     imageType=NoImage
     mirroring=NoMirror
     resourceName=None
     resourceObject=/CorruptionGameplay/UI/ReticleExtensions/Art/M_UI_PwrUp_Glow_Boost
     tiling=NoTile
     imageSize=<24 24>
     margin=<ltrb 0.5 0.5 0.5 0.5>
     outlineSettings {
      bUseBrushTransparency=false
      roundingType=HalfHeightRadius
      width=0
      color {colorUseRule=UseColor_Specified, specifiedColor=<rgba 0 0 0 0>}
      cornerRadii=<0 0 0 0>
     }
     tintColor {colorUseRule=UseColor_Specified, specifiedColor=<rgba 1 1 1 1>}
     uVRegion {bIsValid=false, max=<0 0>, min=<0 0>}
    }
    renderTransform {angle=0, scale=<2.5 2.5>, shear=<0 0>, translation=<0 0>}
   }
   slot {class=OverlaySlot, properties {horizontalAlignment=HAlign_Center, verticalAlignment=VAlign_Center}}
  }
  BoostCircleBacking {
   widgetClassPath=&Image
   bIsVariable=true
   parent=BoostIndicator
   widget=BoostCircleBacking
   properties {
    visibility=SelfHitTestInvisible
    brush {
     bIsDynamicallyLoaded=false
     drawAs=Image
     imageType=NoImage
     mirroring=NoMirror
     resourceName=None
     resourceObject=/CorruptionGameplay/UI/ReticleExtensions/Art/M_UI_PwrUp_Circle_Boost
     tiling=NoTile
     imageSize=<24 24>
     margin=<ltrb 0 0 0 0>
     outlineSettings {
      bUseBrushTransparency=false
      roundingType=HalfHeightRadius
      width=0
      color {colorUseRule=UseColor_Specified, specifiedColor=<rgba 0 0 0 0>}
      cornerRadii=<0 0 0 0>
     }
     tintColor {colorUseRule=UseColor_Specified, specifiedColor=<rgba 1 1 1 1>}
     uVRegion {bIsValid=false, max=<0 0>, min=<0 0>}
    }
   }
   slot {class=OverlaySlot, properties {horizontalAlignment=HAlign_Center, verticalAlignment=VAlign_Center}}
  }
  Ring {
   widgetClassPath=&Image
   bIsVariable=true
   parent=BoostIndicator
   widget=Ring
   properties {
    renderOpacity=0
    visibility=SelfHitTestInvisible
    brush {
     bIsDynamicallyLoaded=false
     drawAs=Image
     imageType=NoImage
     mirroring=NoMirror
     resourceName=None
     resourceObject=/CorruptionGameplay/UI/ReticleExtensions/Art/M_UI_PwrUp_Circle_Warning_Ring_Boost
     tiling=NoTile
     imageSize=<32 32>
     margin=<ltrb 0 0 0 0>
     outlineSettings {
      bUseBrushTransparency=false
      roundingType=HalfHeightRadius
      width=0
      color {colorUseRule=UseColor_Specified, specifiedColor=<rgba 0 0 0 0>}
      cornerRadii=<0 0 0 0>
     }
     tintColor {colorUseRule=UseColor_Specified, specifiedColor=<rgba 1 1 1 1>}
     uVRegion {bIsValid=false, max=<0 0>, min=<0 0>}
    }
   }
   slot {class=OverlaySlot, properties {horizontalAlignment=HAlign_Center, verticalAlignment=VAlign_Center}}
  }
  PulseRinge {
   widgetClassPath=&Image
   bIsVariable=true
   parent=BoostIndicator
   widget=PulseRinge
   properties {
    renderOpacity=0
    visibility=SelfHitTestInvisible
    brush {
     bIsDynamicallyLoaded=false
     drawAs=Image
     imageType=NoImage
     mirroring=NoMirror
     resourceName=None
     resourceObject=null
     tiling=NoTile
     imageSize=<24 24>
     margin=<ltrb 0 0 0 0>
     outlineSettings {
      bUseBrushTransparency=false
      roundingType=HalfHeightRadius
      width=0
      color {colorUseRule=UseColor_Specified, specifiedColor=<rgba 0 0 0 0>}
      cornerRadii=<0 0 0 0>
     }
     tintColor {colorUseRule=UseColor_Specified, specifiedColor=<rgba 1 1 1 1>}
     uVRegion {bIsValid=false, max=<0 0>, min=<0 0>}
    }
   }
   slot {class=OverlaySlot, properties {horizontalAlignment=HAlign_Center, verticalAlignment=VAlign_Center}}
  }
  Arrow {
   widgetClassPath=&Image
   bIsVariable=true
   parent=BoostIndicator
   widget=Arrow
   properties {
    visibility=SelfHitTestInvisible
    brush {
     bIsDynamicallyLoaded=false
     drawAs=Image
     imageType=NoImage
     mirroring=NoMirror
     resourceName=None
     resourceObject=&M_UI_PwrUp_BoostIcon_master_Inst
     tiling=NoTile
     imageSize=<18 18>
     margin=<ltrb 0 0 0 0>
     outlineSettings {
      bUseBrushTransparency=false
      roundingType=HalfHeightRadius
      width=0
      color {colorUseRule=UseColor_Specified, specifiedColor=<rgba 0 0 0 0>}
      cornerRadii=<0 0 0 0>
     }
     tintColor {colorUseRule=UseColor_Specified, specifiedColor=<rgba 1 1 1 1>}
     uVRegion {bIsValid=false, max=<0 0>, min=<0 0>}
    }
   }
   slot {class=OverlaySlot, properties {horizontalAlignment=HAlign_Center, verticalAlignment=VAlign_Center}}
  }
  LightningFX {
   widgetClassPath=&Overlay
   bIsVariable=true
   parent=Overlay_Body
   widget=LightningFX
   properties {visibility=Collapsed, renderTransform {angle=-30, scale=<1 1>, shear=<0 0>, translation=<0 0>}}
   slot {class=OverlaySlot, properties {horizontalAlignment=HAlign_Fill, verticalAlignment=VAlign_Fill}}
  }
  Lightning1 {
   widgetClassPath=&Image
   bIsVariable=true
   parent=LightningFX
   widget=Lightning1
   properties {
    renderOpacity=0
    visibility=SelfHitTestInvisible
    brush {
     bIsDynamicallyLoaded=false
     drawAs=Image
     imageType=NoImage
     mirroring=NoMirror
     resourceName=None
     resourceObject=&M_UI_PwrUp_Lightning
     tiling=NoTile
     imageSize=<32 32>
     margin=<ltrb 0 0 0 0>
     outlineSettings {
      bUseBrushTransparency=false
      roundingType=HalfHeightRadius
      width=0
      color {colorUseRule=UseColor_Specified, specifiedColor=<rgba 0 0 0 0>}
      cornerRadii=<0 0 0 0>
     }
     tintColor {colorUseRule=UseColor_Specified, specifiedColor=<rgba 1 1 1 1>}
     uVRegion {bIsValid=false, max=<0 0>, min=<0 0>}
    }
    renderTransform {angle=0, scale=<1 1>, shear=<0 0>, translation=<0 12>}
   }
   slot {class=OverlaySlot, properties {horizontalAlignment=HAlign_Center, verticalAlignment=VAlign_Bottom}}
  }
  Overlay {
   widgetClassPath=&Overlay
   parent=LightningFX
   widget=Overlay
   properties {renderTransform {angle=-27, scale=<1 1>, shear=<0 0>, translation=<0 0>}}
   slot {class=OverlaySlot, properties {horizontalAlignment=HAlign_Fill, verticalAlignment=VAlign_Fill}}
  }
  Lightning2 {
   widgetClassPath=&Image
   bIsVariable=true
   parent=Overlay
   widget=Lightning2
   properties {
    renderOpacity=0
    visibility=SelfHitTestInvisible
    brush {
     bIsDynamicallyLoaded=false
     drawAs=Image
     imageType=NoImage
     mirroring=NoMirror
     resourceName=None
     resourceObject=/Game/Athena/HUD/XP/Materials/M_UI_PwrUp_Lightning_2
     tiling=NoTile
     imageSize=<32 32>
     margin=<ltrb 0 0 0 0>
     outlineSettings {
      bUseBrushTransparency=false
      roundingType=HalfHeightRadius
      width=0
      color {colorUseRule=UseColor_Specified, specifiedColor=<rgba 0 0 0 0>}
      cornerRadii=<0 0 0 0>
     }
     tintColor {colorUseRule=UseColor_Specified, specifiedColor=<rgba 1 1 1 1>}
     uVRegion {bIsValid=false, max=<0 0>, min=<0 0>}
    }
    renderTransform {angle=0, scale=<1 -1>, shear=<0 0>, translation=<0 12>}
   }
   slot {class=OverlaySlot, properties {horizontalAlignment=HAlign_Center, verticalAlignment=VAlign_Bottom}}
  }
  Overlay_1 {
   widgetClassPath=&Overlay
   parent=Overlay
   widget=Overlay_1
   properties {renderTransform {angle=-27, scale=<1 1>, shear=<0 0>, translation=<0 0>}}
   slot {class=OverlaySlot, properties {horizontalAlignment=HAlign_Fill, verticalAlignment=VAlign_Fill}}
  }
  Lightning3 {
   widgetClassPath=&Image
   bIsVariable=true
   parent=Overlay_1
   widget=Lightning3
   properties {
    renderOpacity=0
    visibility=SelfHitTestInvisible
    brush {
     bIsDynamicallyLoaded=false
     drawAs=Image
     imageType=NoImage
     mirroring=NoMirror
     resourceName=None
     resourceObject=/Game/Athena/HUD/XP/Materials/M_UI_PwrUp_Lightning_3
     tiling=NoTile
     imageSize=<32 32>
     margin=<ltrb 0 0 0 0>
     outlineSettings {
      bUseBrushTransparency=false
      roundingType=HalfHeightRadius
      width=0
      color {colorUseRule=UseColor_Specified, specifiedColor=<rgba 0 0 0 0>}
      cornerRadii=<0 0 0 0>
     }
     tintColor {colorUseRule=UseColor_Specified, specifiedColor=<rgba 1 1 1 1>}
     uVRegion {bIsValid=false, max=<0 0>, min=<0 0>}
    }
    renderTransform {angle=0, scale=<-1 1>, shear=<0 0>, translation=<0 12>}
   }
   slot {class=OverlaySlot, properties {horizontalAlignment=HAlign_Center, verticalAlignment=VAlign_Bottom}}
  }
  TutorialHighlight {widgetClassPath=&TutorialHighlight}
 ]
}
